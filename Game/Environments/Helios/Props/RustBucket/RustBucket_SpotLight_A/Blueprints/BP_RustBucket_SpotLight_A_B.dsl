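; Construction script for a RustBucket spotlight prop Blueprint that applies building texture data to its meshes.
; After calling the parent script, it reads the texture data array and continues only if it has entries.
; The first entry's override material goes on slot 0 of both the main static mesh and the Yacht_Spotlight_Can mesh.
; Its tint color is then passed to the can through SetColoronSecondaryMesh.
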(fn ConstructionScript
  (|CallParentFunction)
  (bind _outtexturedataarray (Building|GetBuildingTextureData))
  (if (> (Utilities|Array|Length _outtexturedataarray) 0)
    (bind _outtexturedataarray_2 (Building|GetBuildingTextureData))
    (Utilities|FlowControl|Sequence
      (:then_0
        (Rendering|Material|SetMaterial (Variables|Debug|GetStaticMeshComponent) 0 (Variables|/Script/FortniteGame.BuildingTextureData|GetOverrideMaterial (Utilities|Array|Get(acopy) _outtexturedataarray_2)))
        (Rendering|Material|SetMaterial (Variables|Default|GetYacht_Spotlight_Can) 0 (Variables|/Script/FortniteGame.BuildingTextureData|GetOverrideMaterial (Utilities|Array|Get(acopy) _outtexturedataarray_2))))
      (:then_1
        (Default|SetColoronSecondaryMesh (Math|Conversions|ToLinearColor(Color) (Variables|/Script/FortniteGame.BuildingTextureData|GetTintColor (Utilities|Array|Get(acopy) _outtexturedataarray_2))) (Variables|Default|GetYacht_Spotlight_Can))))))
