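# Default Lit material for the RustBucket spotlight prop, with a tinted, mask-driven glow and a vertex-animated swivel.
# Diffuse, normal, specular and emissive-mask textures supply the surface, with Color and Glow parameters scaling the emission.
# World position offset rotates vertices around the instance's Y axis and origin, using offset time through a sine with speed and angle parameters.
# A static switch, quality levels and ES3.1 fallback can zero the motion.
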Material /Game/Environments/Helios/Props/RustBucket/RustBucket_SpotLight_A/Materials/MI_Rustbucket_PanningSpotlight_B
graph {
 properties {bAutomaticallySetUsageInEditor=false, pixelDepthOffsetMode=PDOM_Legacy, usedShadingModels=Default Lit}
 nodes [
  n1 = TextureSample() {
   samplerType=SAMPLERTYPE_Masks
   texture=/Game/Environments/Helios/Props/RustBucket/RustBucket_SpotLight_A/Textures/T_RustBucket_Spotlight_A_A_E
  }
  LinearInterpolate(
   A=Multiply(
    A=TextureSample() {texture=/Game/Environments/Helios/Props/RustBucket/RustBucket_SpotLight_A/Textures/T_RustBucket_Spotlight_A_A_D}.R
    B=Constant() {r=1}
   )
   Alpha=@n1.G
  ) {constB=0.2}
  n2 = VectorParameter() {parameterName=Color}
  n3 = TextureSample() {
   samplerType=SAMPLERTYPE_Masks
   texture=&T_RustBucket_Spotlight_A_A_E
  }
  Multiply(A=Multiply(A=Multiply(A=@n2.RGB), B=@n3.R), B=ScalarParameter() {defaultValue=2, parameterName=Glow})
  n4 = NamedRerouteUsage()
  n5 = RotateAboutAxis(NormalizedRotationAxis=NamedRerouteUsage(), RotationAngle=NamedRerouteUsage(), PivotPoint=Constant3Vector(), Position=@n4)
  NamedRerouteDeclaration(None=Normalize(VectorInput=Add(A=@n5, B=@n4))) {name=Cone Z Rotated, nodeColor=<rgba 1 0 0.375 1>}
  n6 = NamedRerouteUsage()
  n7 = NamedRerouteUsage()
  NamedRerouteDeclaration(None=Normalize(VectorInput=CrossProduct(A=@n6, B=@n7))) {name=Cone X Rotated, nodeColor=<rgba 1 0 0.375 1>}
  n8 = NamedRerouteUsage()
  n9 = NamedRerouteUsage()
  NamedRerouteDeclaration(None=Normalize(VectorInput=CrossProduct(A=@n8, B=@n9))) {name=Cone Y Rotated, nodeColor=<rgba 1 0 0.375 1>}
  n10 = Normalize(VectorInput=Transform(None=Constant3Vector() {constant=<rgba 1 0 0 0>}) {transformSourceType=TRANSFORMSOURCE_Instance})
  NamedRerouteDeclaration(None=ShaderStageSwitch(PixelShader=VertexInterpolator(VS=@n10).PS, VertexShader=@n10)) {name=Cone X BeforeRotation, nodeColor=<rgba 1 0.6797 0 1>}
  n11 = Time()
  n12 = MaterialFunctionCall(Float3=NamedRerouteUsage()) {materialFunction=&BreakOutFloat3Components}
  n13 = Add(A=@n12.R, B=@n12.R)
  n14 = Multiply(A=Frac(None=Add(A=@n13, B=@n12.B))) {constB=33.3}
  n15 = Sine(None=Multiply(A=Add(A=@n11, B=@n14), B=ScalarParameter() {defaultValue=0.1, parameterName=WPOAnim_Speed}))
  n16 = ScalarParameter() {defaultValue=0.1, parameterName=WPOAnim_MaxAngle}
  NamedRerouteDeclaration(None=Frac(None=Multiply(A=@n15, B=@n16))) {name=Anim Rotation, nodeColor=<rgba 0 1 0.5078 1>}
  n17 = Normalize(VectorInput=Transform(None=Constant3Vector() {constant=<rgba 0 1 0 0>}) {transformSourceType=TRANSFORMSOURCE_Instance})
  n18 = VertexInterpolator(VS=@n17)
  NamedRerouteDeclaration(None=NamedRerouteDeclaration(None=ShaderStageSwitch(PixelShader=@n18.PS, VertexShader=@n17)) {name=Cone Y BeforeRotation, nodeColor=<rgba 1 0.6797 0 1>}) {name=RotationAxis, nodeColor=<rgba 0.0312 1 0 1>}
  n19 = Normalize(VectorInput=Transform(None=Constant3Vector() {constant=<rgba 0 0 1 0>}) {transformSourceType=TRANSFORMSOURCE_Instance})
  NamedRerouteDeclaration(None=ShaderStageSwitch(PixelShader=VertexInterpolator(VS=@n19).PS, VertexShader=@n19)) {name=Cone Z BeforeRotation, nodeColor=<rgba 1 0.6797 0 1>}
  n20 = Constant3Vector()
  n21 = Reroute()
  n22 = RotateAboutAxis(
   NormalizedRotationAxis=NamedRerouteUsage()
   RotationAngle=NamedRerouteUsage()
   PivotPoint=NamedRerouteDeclaration(None=TransformPosition(None=@n20) {transformSourceType=TRANSFORMPOSSOURCE_Instance, transformType=TRANSFORMPOSSOURCE_World}) {name=Cone Origin, nodeColor=<rgba 0.0638 0.4605 1 1>}
   Position=Add(A=WorldPosition().XYZ, B=@n21)
  )
  n23 = Constant3Vector()
  QualitySwitch(
   Default=FeatureLevelSwitch(Default=StaticSwitchParameter(True=@n22, False=@n23) {defaultValue=true, parameterName=UseWpoAnimation}, ES3_1=@n23)
   Low=@n23
   High=@n23
   Medium=@n23
  )
  Reroute()
  StaticSwitchParameter(True=Add(A=@n21, B=@n22), False=@n21) {defaultValue=true, parameterName=UseWpoAnimation}
  TextureSample() {
   samplerType=SAMPLERTYPE_Normal
   texture=/Game/Environments/Helios/Props/RustBucket/RustBucket_SpotLight_A/Textures/T_RustBucket_Spotlight_A_A_N
  }
  TextureSample() {texture=&T_RustBucket_Spotlight_A_A_D}
  TextureSample() {
   samplerType=SAMPLERTYPE_Masks
   texture=/Game/Environments/Helios/Props/RustBucket/RustBucket_SpotLight_A/Textures/T_RustBucket_Spotlight_A_A_S
  }
 ]
}
